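# Unlit translucent material for a toon-spark tyre streak in the drift trail vehicle cosmetic, enabled for Niagara ribbons and mesh particles.
# Two panned samples of MaskTexture combine with scrolling sphere masks and gradient stripes to form animated, broken-up streaks.
# UV gradients shape the streak and fade its edges, with particle colour tinting emissive and particle alpha scaling opacity.
# Shared functions apply depth fade and emissive adjustments; some UV math feeds nothing visible.
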Material /VehicleCosmetics/DriftTrail/DriftTrail_ToonSparks/Material/M_ToonStreak_Tyre1
graph {
 properties {
  bUsedWithNiagaraMeshParticles=true
  bUsedWithNiagaraRibbons=true
  blendMode=BLEND_Translucent
  pixelDepthOffsetMode=PDOM_Legacy
  shadingModel=MSM_Unlit
  usedShadingModels=Unlit
 }
 nodes [
  n1 = TextureCoordinate()
  n2 = Constant() {r=1}
  Add(
   A=@n1
   B=Multiply(
    A=MaterialFunctionCall(Speed=@n2) {materialFunction=&TimeWithSpeedVariable}.Frac Time
    B=Constant2Vector() {g=1}
   )
  )
  MaterialExpressionVertexColor_0 = ParticleColor()
  n3 = Constant() {r=5}
  n4 = TextureCoordinate()
  n5 = ComponentMask(None=@n4) {r=true}
  n6 = Constant() {r=3}
  n7 = OneMinus(None=@n5)
  n8 = ComponentMask(None=@n4) {g=true}
  n9 = Multiply(A=@n8, B=OneMinus(None=@n8))
  n10 = Multiply(A=@n7, B=Multiply(A=@n9, B=Constant() {r=5}))
  n11 = Constant() {r=2}
  n12 = TextureCoordinate() {coordinateIndex=1, vTiling=0.5}
  n13 = Constant() {r=0.5}
  n14 = TextureSampleParameter2D(UVs=Panner(Coordinate=Panner(Coordinate=@n12, Time=@n13) {speedY=0.5}) {speedY=-0.5}) {parameterName=MaskTexture, texture=/VehicleCosmetics/DriftTrail/DriftTrail_ToonSparks/Textures/T_VCVFX_Trail_ToonSparks_supersonic02Tyre_RGB}
  n15 = TextureCoordinate() {coordinateIndex=1, uTiling=0.5}
  n16 = TextureSampleParameter2D(UVs=Panner(Coordinate=Panner(Coordinate=@n15, Time=@n13) {speedX=0.5}) {speedY=-1.2}) {parameterName=MaskTexture, texture=&T_VCVFX_Trail_ToonSparks_supersonic02Tyre_RGB}
  n17 = Multiply(A=@n14.G, B=@n16.G)
  n18 = TextureSample(UVs=Panner() {speedX=-0.5, speedY=0.5}) {texture=/DealershipSharedLibrary/FX/Textures/Gradient_Masks/T_VCVFX_SphereMaskTimeOffset_A}
  n19 = TextureSample(UVs=TextureCoordinate() {vTiling=1.25}) {texture=/DealershipSharedLibrary/FX/Textures/Gradient_Masks/T_VCVFX_Trail_ToonSparks_GradientStripeTyre_A}
  n20 = Multiply(A=Add(A=@n16.R, B=@n16.R), B=Multiply(A=@n18.R, B=@n19.RGB))
  MaterialExpressionConstantClamp_1 = Saturate(None=Subtract(A=Add(A=@n17, B=@n20), B=@n20))
  n21 = Multiply(A=@n16.R, B=@n14.G)
  n22 = Panner() {speedX=0.5, speedY=0.35}
  n23 = Multiply(
   A=Multiply(A=TextureSample(UVs=@n22) {texture=&T_VCVFX_SphereMaskTimeOffset_A}.G, B=@n19.RGB)
   B=Add(A=@n14.R, B=@n16.R)
  )
  MaterialExpressionConstantClamp_2 = Saturate(None=Subtract(A=Add(A=@n21, B=@n23), B=@n20))
  n24 = Add(A=Multiply(A=@MaterialExpressionConstantClamp_1, B=@MaterialExpressionConstantClamp_2), B=Multiply(A=@n14.B, B=@n18.B))
  n25 = Multiply(A=Multiply(A=@n7, B=@n9), B=@n8)
  n26 = Constant() {r=0.3}
  MaterialExpressionConstantClamp_4 = Saturate(None=Add(A=Multiply(A=@n11, B=@n24), B=Multiply(A=@n25, B=@n26)))
  n27 = Multiply(A=Multiply(A=Multiply(A=@n3, B=@n5), B=Multiply(A=@n6, B=@n10)), B=@MaterialExpressionConstantClamp_4)
  n28 = TextureSample(UVs=Panner(Coordinate=TextureCoordinate() {vTiling=0.2}, Time=Constant() {r=0.5})) {texture=&T_VCVFX_Trail_ToonSparks_GradientStripeTyre_A}
  MaterialExpressionConstantClamp_3 = Saturate(None=Multiply(A=Multiply(A=@n27, B=@n28.RGB), B=Constant() {r=8}))
  n29 = Sine(None=ComponentMask(None=TextureCoordinate()) {g=true}) {period=2}
  MaterialFunctionCall(
   Opacity In=Multiply(A=Multiply(A=@MaterialExpressionVertexColor_0.A, B=@MaterialExpressionConstantClamp_3), B=Power(Base=@n29) {constExponent=5})
   DepthFade Distance=Constant()
  ) {materialFunction=&MF_VCVFX_DepthFade}
  n30 = Constant() {r=3}
  n31 = Add(A=@n27, B=@n9)
  MaterialFunctionCall(Emissive=Multiply(A=@MaterialExpressionVertexColor_0.RGB, B=Multiply(A=@n30, B=@n31))) {materialFunction=&MF_FX_EmissiveAdjustments}
  n32 = ComponentMask(None=@n1) {r=true}
  n33 = Sine(None=@n32) {period=2}
  n34 = Sine(None=ComponentMask(None=@n1) {g=true}) {period=2}
  n35 = Multiply(
   A=Multiply(A=Power(Base=@n33) {constExponent=1}, B=Power(Base=@n34) {constExponent=4})
   B=Power(Base=OneMinus(None=@n32)) {constExponent=0.5}
  )
  Multiply(A=@n35) {constB=2}
  Power(Base=@n35) {constExponent=0.5}
 ]
}
